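# Niagara system for the radial impact effect of the IBlade melee weapon, under the weapon gameplay cues.
# It layers thirteen named emitters: flash, shell and floor bursts, shockwave, corona, distant glow, darkness, light, dust, debris, embers, initial sparks and residual electric arcs.
# The system exposes no user variables and lists only SystemState as a module, with no renderer details for any emitter.
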NiagaraSystem /FNGameplayCues/Weapons/FORT_Melee/Effects/Blade/NS_IBlade_Radial_Impact
dependencies {usedDataInterfaces [], usedDynamicInputs [], usedModules [&SystemState]}
emitters {
 Corona {rendererClasses [], renderers []}
 Darkness {rendererClasses [], renderers []}
 DistantGlow {rendererClasses [], renderers []}
 Dust {rendererClasses [], renderers []}
 ElectricArcsResidual {rendererClasses [], renderers []}
 Embers {rendererClasses [], renderers []}
 Flash {rendererClasses [], renderers []}
 FloorBurst {rendererClasses [], renderers []}
 Light {rendererClasses [], renderers []}
 ShellBurst {rendererClasses [], renderers []}
 ShockWave {rendererClasses [], renderers []}
 Sparks_Initial {rendererClasses [], renderers []}
 debris {rendererClasses [], renderers []}
}
summary {
 systemName=NS_IBlade_Radial_Impact
 emitters [
  {emitterName=ShellBurst, rendererClasses []}
  {emitterName=FloorBurst, rendererClasses []}
  {emitterName=Corona, rendererClasses []}
  {emitterName=DistantGlow, rendererClasses []}
  {emitterName=Darkness, rendererClasses []}
  {emitterName=Light, rendererClasses []}
  {emitterName=Dust, rendererClasses []}
  {emitterName=ShockWave, rendererClasses []}
  {emitterName=debris, rendererClasses []}
  {emitterName=Embers, rendererClasses []}
  {emitterName=ElectricArcsResidual, rendererClasses []}
  {emitterName=Flash, rendererClasses []}
  {emitterName=Sparks_Initial, rendererClasses []}
 ]
 userVariables []
}
